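# Conversation database for the NPC Sabina, a secrets dealer who greets players and leads into shared quest and service dialogue.
# Her entry tag opens one speech node with a default line and contextual alternatives.
# These alternatives check for a first-ever conversation or specific player character cosmetics, with a priority curve set only on the first-meeting line.
# The speech node then links to the remote NPCLibrary Quests and BasicServices conversation entries.
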ConversationDatabase /NPCLibrary/NPCs/Sabina/NPCLibrary_Conversation_Sabina
entry_points [{tag=Conversation.Entry.Sabina, destinations [#1]}]
nodes [
 ConversationEntryPointNode_7 {
  id=#1
  class=ConversationEntryPointNode
  outputs [#2]
  properties {entryTag {tagName=Conversation.Entry.Sabina}}
 }
 Speech_C_4 {
  id=#2
  class=Speech_C
  outputs [#3, #4]
  properties {
   nodeName=""
   generalConfig {
    defaultMessage=I know everyone's secrets. Even yours.
    contextualMessages [
     {
      message=I'm in the intel business. Got secrets to sell?
      contextRequirements [{requirement {class=FortControllerRequirement_IsFirstEverConversationWithNPC}}]
      priority {curve {rowName=Tandem.Speech.Priority.FirstEverInteraction}}
     }
     {
      message=I know what you're looking for.
      contextRequirements [{requirement {
         class=FortControllerRequirement_HasCID
         properties {allowedCIds [/BRCosmetics/Athena/Items/Cosmetics/Characters/CID_A_135_Athena_Commando_F_StreetFashionSummer]}
        }}]
      priority {value=0, curve {curveTable=null, rowName=None}}
     }
     {
      message=My prices are high. My information is good.
      requirementMatchPolicy=RequireAny
      contextRequirements [{requirement {
         class=FortControllerRequirement_HasCID
         properties {allowedCIds [
           /BRCosmetics/Athena/Items/Cosmetics/Characters/CID_155_Athena_Commando_F_Gumshoe
           /BRCosmetics/Athena/Items/Cosmetics/Characters/CID_159_Athena_Commando_M_GumshoeDark
           /BRCosmetics/Athena/Items/Cosmetics/Characters/CID_154_Athena_Commando_M_Gumshoe
          ]}
        }}]
      priority {value=0, curve {curveTable=null, rowName=None}}
     }
     {
      message=Who are you without the IO signing your checks?
      requirementMatchPolicy=RequireAny
      contextRequirements [{requirement {
         class=FortControllerRequirement_HasCID
         properties {allowedCIds [
           /BRCosmetics/Athena/Items/Cosmetics/Characters/CID_A_202_Athena_Commando_F_Division
           /BRCosmetics/Athena/Items/Cosmetics/Characters/CID_A_398_Athena_Commando_F_NeonCatSpeed
           /BRCosmetics/Athena/Items/Cosmetics/Characters/CID_A_370_Athena_Commando_M_OrderGuard
          ]}
        }}]
      priority {value=0, curve {curveTable=null, rowName=None}}
     }
     {
      message=Never strike a deal unless you know the cost.
      contextRequirements [{requirement {class=FortControllerRequirement_HasCID, properties {allowedCIds [&Character_Bites]}}}]
      priority {value=0, curve {curveTable=null, rowName=None}}
     }
     {
      message="Oh, you come from a happy family? Good for you."
      contextRequirements [{requirement {class=FortControllerRequirement_HasCID, properties {allowedCIds [&Character_ChillCat]}}}]
      priority {value=0, curve {curveTable=null, rowName=None}}
     }
     {
      message=I've got contacts in the Academy. Your friends are safe.
      contextRequirements [{requirement {class=FortControllerRequirement_HasCID, properties {allowedCIds [&Character_PinkSpike]}}}]
      priority {value=0, curve {curveTable=null, rowName=None}}
     }
     {
      message=Revenge won't make you feel any better.
      contextRequirements [{requirement {class=FortControllerRequirement_HasCID, properties {allowedCIds [&Character_BadBear]}}}]
      priority {value=0, curve {curveTable=null, rowName=None}}
     }
     {
      message=I've been hearing your name in quiet places for a long time.
      contextRequirements [{requirement {class=FortControllerRequirement_HasCID, properties {allowedCIds [&Character_TheHerald]}}}]
      priority {value=0, curve {curveTable=null, rowName=None}}
     }
     {
      message="You're into spiders, huh? Yeah, me too."
      contextRequirements [{requirement {class=FortControllerRequirement_HasCID, properties {allowedCIds [&Character_RoseDust]}}}]
      priority {value=0, curve {curveTable=null, rowName=None}}
     }
    ]
   }
  }
 }
 ConversationLinkNode_0 {id=#4, class=ConversationLinkNode, properties {remoteEntryTag {tagName=Conversation.Entry.BasicServices}}}
 ConversationLinkNode_1 {id=#3, class=ConversationLinkNode, properties {remoteEntryTag {tagName=Conversation.Entry.NPCLibrary.Quests}}}
]
